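# Unlit Material Gym example material that demonstrates virtual-plane UV mapping with a color grid texture.
# MF_VirutalPlaneUVs builds UVs from a plane centered on the actor's world position, with constant Y normal and X axis vectors.
# A UVScale parameter feeds both scale components, and the resulting V coordinate is negated before sampling the UV Editor color grid.
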Material /Game/Global/Materials/MaterialGym/Materials/MaterialElements/UVs/VirtualPlane/M_VirtualPlane
graph {
 expressions {
  MaterialExpressionActorPositionWS_0 {class=MaterialExpressionActorPositionWS, inputs [], properties {}}
  MaterialExpressionConstant2Vector_0 {class=MaterialExpressionConstant2Vector, properties {g=-1, r=1}}
  MaterialExpressionConstant3Vector_0 {class=MaterialExpressionConstant3Vector, properties {constant=<rgba 0 1 0 0>}}
  MaterialExpressionConstant3Vector_1 {class=MaterialExpressionConstant3Vector, properties {constant=<rgba 1 0 0 0>}}
  MaterialExpressionConvert_0 {
   class=MaterialExpressionConvert
   inputs [
    None {from=MaterialExpressionScalarParameter_0}
    None {from=MaterialExpressionScalarParameter_0}
   ]
   properties {
    nodeName=Make Float2
    convertInputs [
     {
      type=Scalar
      expressionInput {
       expression=&self:MaterialExpressionScalarParameter_0
       inputName=None
       mask=0
       maskA=0
       maskB=0
       maskG=0
       maskR=0
       outputIndex=0
      }
     }
     {
      type=Scalar
      expressionInput {
       expression=&self:MaterialExpressionScalarParameter_0
       inputName=None
       mask=0
       maskA=0
       maskB=0
       maskG=0
       maskR=0
       outputIndex=0
      }
     }
    ]
    convertOutputs [{type=Vector2}]
   }
  }
  MaterialExpressionMaterialFunctionCall_5 {
   class=MaterialExpressionMaterialFunctionCall
   inputs [
    Plane Normal Axis {from=MaterialExpressionConstant3Vector_0}
    Plane X Axis {from=MaterialExpressionConstant3Vector_1}
    Plane Center {from=MaterialExpressionActorPositionWS_0}
    UV Scale {from=MaterialExpressionConvert_0}
   ]
   properties {materialFunction=&/MF_VirutalPlaneUVs}
  }
  MaterialExpressionMultiply_5 {
   class=MaterialExpressionMultiply
   inputs [
    A {from=MaterialExpressionMaterialFunctionCall_5, output=UVs}
    B {from=MaterialExpressionConstant2Vector_0}
   ]
  }
  MaterialExpressionScalarParameter_0 {class=MaterialExpressionScalarParameter, properties {defaultValue=140, parameterName=UVScale}}
  MaterialExpressionTextureSample_0 {
   class=MaterialExpressionTextureSample
   inputs [
    UVs {from=MaterialExpressionMultiply_5}
    Tex {from=""}
    Apply View MipBias {from=""}
   ]
   properties {texture=/UVEditor/Textures/UVEditorColorGrid_Color}
  }
 }
 properties {bAutomaticallySetUsageInEditor=false, shadingModel=MSM_Unlit, usedShadingModels=Unlit}
}
